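# Post-process-style Anim Blueprint for the DroneGravy emote gadget that adds rigid body physics to its incoming pose, apparently for a cord.
# The input pose is cached as CordPose and fed through a RigidBody node that overrides world gravity.
# Each update reads the DisablePhysics animation curve; at 1 or above, a bool blend outputs the unsimulated pose instead, letting animations switch physics off.
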AnimBlueprint /CosmeticCompanions/Assets/Props/Emotes/DroneGravy/Meshes/ABP_DroneGravy_Gadget
target_skeleton=&/SK_DroneGravy_Gadget_Skeleton
graphs [
 AnimGraph {nodes [
   n1 = RigidBody(ComponentPose=Local To Component(LocalPose=Use cached pose 'CordPose'())) {bOverrideWorldGravity=true, alphaBoolBlend {blendOption=Linear}}
   Output Pose(Result=Blend Poses by bool(
     BlendPose_0=Use cached pose 'CordPose'()
     BlendPose_1=Component To Local(ComponentPose=@n1)
     bActiveValue=Get Disable Physics() -> (Disable Physics:bool)
    ))
   Save cached pose 'CordPose'(Pose=Input Pose()) {cacheName=CordPose}
  ]}
 EventGraph {nodes [
   Event BlueprintUpdateAnimation()
   > Set Disable Physics(Disable Physics:bool="float >= float"(A=GetCurveValue(CurveName=DisablePhysics), B=1)) -> (Output_Get:bool)
   TryGetPawnOwner()
  ]}
]
variables [Disable Physics {type=boolean}]
