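# Physics asset for the gadget mesh of the DroneGravy emote prop, defining physics bodies for its cord bone chain.
# Eight identical small boxes cover cord_0 through cord_7.
# Four constraints link cord_3 through cord_7 with tight twist limits and moderate swing, widest at the cord_4 to cord_5 joint.
# The bodies cord_0 through cord_2 have no constraints listed.
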PhysicsAsset /CosmeticCompanions/Assets/Props/Emotes/DroneGravy/Meshes/PA_DroneGravy_Gadget
bodies {
 cord_0 [{shapeType=Box, extentX=4, extentY=2, extentZ=2}]
 cord_1 [{shapeType=Box, extentX=4, extentY=2, extentZ=2}]
 cord_2 [{shapeType=Box, extentX=4, extentY=2, extentZ=2}]
 cord_3 [{shapeType=Box, extentX=4, extentY=2, extentZ=2}]
 cord_4 [{shapeType=Box, extentX=4, extentY=2, extentZ=2}]
 cord_5 [{shapeType=Box, extentX=4, extentY=2, extentZ=2}]
 cord_6 [{shapeType=Box, extentX=4, extentY=2, extentZ=2}]
 cord_7 [{shapeType=Box, extentX=4, extentY=2, extentZ=2}]
}
constraints [
 {bone1Name=cord_4, bone2Name=cord_5, swing1LimitDegrees=90, swing2LimitDegrees=90, twistLimitDegrees=1}
 {bone1Name=cord_6, bone2Name=cord_5, swing1LimitDegrees=45, swing2LimitDegrees=45, twistLimitDegrees=1}
 {bone1Name=cord_6, bone2Name=cord_7, swing1LimitDegrees=45, swing2LimitDegrees=45, twistLimitDegrees=5}
 {bone1Name=cord_4, bone2Name=cord_3, swing1LimitDegrees=45, swing2LimitDegrees=45, twistLimitDegrees=5}
]
